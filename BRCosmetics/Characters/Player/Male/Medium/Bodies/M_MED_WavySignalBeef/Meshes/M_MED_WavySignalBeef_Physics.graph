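# Physics asset for the M_MED_WavySignalBeef male player body cosmetic, set up for a dangling sword on a belt.
# A constraint chain runs from dyn_swordBeltBase through two belt bones to the sword, with tight twist and narrow swing limits.
# Capsules cover each chain bone, with three forming the sword.
# Other capsules cover only the pelvis, spine_02, left thigh and left calf, apparently as colliders for the swinging sword.
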PhysicsAsset /BRCosmetics/Characters/Player/Male/Medium/Bodies/M_MED_WavySignalBeef/Meshes/M_MED_WavySignalBeef_Physics
bodies {
 calf_l [{shapeType=Capsule, length=33.268, radius=7.7335, center=<-23.0487 -1.5204 0>, rotation=<rot 90 0 0>}]
 dyn_swordBeltBase [{shapeType=Capsule, length=0.505, radius=0.505, shapeName=dyn_swordBeltBase_capsule, rotation=<rot 0 -90 0>}]
 dyn_swordBelt_01 [{shapeType=Capsule, length=2.6026, radius=2.311, shapeName=dyn_swordBelt_01_capsule, rotation=<rot 90 0 0>}]
 dyn_swordBelt_02 [{shapeType=Capsule, length=2.6026, radius=2.311, shapeName=dyn_swordBelt_01_capsule, center=<1.0697 0 0>, rotation=<rot 90 0 0>}]
 dyn_sword_01 [
  {shapeType=Capsule, length=78.4466, radius=2.143, shapeName=dyn_sword_01_capsule, center=<17 0 0>, rotation=<rot 90 0 0>}
  {shapeType=Capsule, length=15.5, radius=2.5, shapeName=dyn_sword_01_capsule, center=<-23.8482 -1.0621 0>, rotation=<rot -90 0.0001 0>}
  {shapeType=Capsule, length=15, radius=3, shapeName=dyn_sword_01_capsule, center=<-15.0482 3.7982 0>, rotation=<rot -90 0.0001 0>}
 ]
 pelvis [{shapeType=Capsule, length=3.5, radius=14.4621, center=<-4.2989 0 0>, rotation=<rot 0 -0.0003 0>}]
 spine_02 [{shapeType=Capsule, length=4, radius=13.4621, center=<2.4326 0 0>}]
 thigh_l [{shapeType=Capsule, length=23.5962, radius=8.6498, center=<-21.8582 0 0>, rotation=<rot 90 0 0>}]
}
constraints [
 {bone1Name=dyn_swordBelt_01, bone2Name=dyn_swordBeltBase, swing1LimitDegrees=5, swing2LimitDegrees=10, twistLimitDegrees=1}
 {bone1Name=dyn_swordBelt_02, bone2Name=dyn_swordBelt_01, swing1LimitDegrees=10, swing2LimitDegrees=20, twistLimitDegrees=1}
 {bone1Name=dyn_sword_01, bone2Name=dyn_swordBelt_02, swing1LimitDegrees=10, swing2LimitDegrees=20, twistLimitDegrees=1}
]
