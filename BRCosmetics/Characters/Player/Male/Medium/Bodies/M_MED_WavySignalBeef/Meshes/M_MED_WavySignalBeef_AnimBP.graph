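# Animation Blueprint for the WavySignalBeef male body cosmetic, copying the parent mesh pose and adding secondary motion, notably for its sword and belt.
# RigidBody physics simulated in spine_01 space feeds two sword-offset bone rotations, then a Control Rig driven by skydiving and targeting state.
# DefaultSlot animation overrides sword and belt bones, weighted by a front-end idle curve.
# The event graph damps targeting, crouch-idle and long-backbling alphas, though only targeting visibly feeds the AnimGraph.
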AnimBlueprint /BRCosmetics/Characters/Player/Male/Medium/Bodies/M_MED_WavySignalBeef/Meshes/M_MED_WavySignalBeef_AnimBP
target_skeleton=&/M_MED_WavySignalBeef_Skeleton
graphs [
 AnimGraph {nodes [
   n1 = RigidBody(ComponentPose=Local To Component(LocalPose=Copy Pose From Mesh() {bUseAttachedParent=true})) {
    overridePhysicsAsset=&/M_MED_WavySignalBeef_Physics
    simulationSpace=BaseBoneSpace
    baseBoneRef {boneName=spine_01}
    simSpaceSettings {dampingAlpha=0, maxAngularAcceleration=25, maxAngularVelocity=5, maxLinearAcceleration=1000, maxLinearVelocity=200, worldAlpha=1}
   }
   n2 = "Transform (Modify) Bone - Bone: dyn_sword_offset"(ComponentPose=@n1, Alpha=1, Translation="0,0,0", Rotation="0,15,0", Scale="1,1,1") {rotationMode=BMM_Replace, rotationSpace=BCS_BoneSpace, boneToModify {boneName=dyn_sword_offset}, rotation {yaw=15}}
   n3 = Control Rig!(
    Source=Component To Local(ComponentPose="Transform (Modify) Bone - Bone: dyn_sword_offset"(ComponentPose=@n2, Alpha=1, Translation="0,0,0", Rotation="0,-20,0", Scale="1,1,1") {rotationMode=BMM_Replace, rotationSpace=BCS_BoneSpace, boneToModify {boneName=dyn_sword_offset}, rotation {yaw=-20}})
    Alpha
    IsSkyDiving:bool=Get bIsSkyDiving()
    IsTargetingAlpha:double=Get IsTargetAlpha() -> (IsTargetAlpha:double)
   ) -> (Pose) {controlRigAssetReference {blueprintRigClass=&/M_MED_WavySignalBeef_CtrlRig.M_MED_WavySignalBeef_CtrlRig_C}}
   n4 = Layered blend per bone(BasePose=@n3, BlendPoses_0=Slot 'DefaultSlot'(), BlendWeights_0=Get customIdleIsPlayingFloat() -> (customIdleIsPlayingFloat:float)) {layerSetup [{branchFilters [
       {boneName=jnt_sword_handle}
       {boneName=dyn_sword_offset}
       {boneName=dyn_sword_01}
       {boneName=dyn_swordBelt_02}
       {boneName=dyn_swordBelt_01}
       {boneName=dyn_swordBeltBase}
      ]}]}
   Output Pose(Result=@n4) {layerGroup=None}
  ]}
 EventGraph {nodes [
   n1 = Event BlueprintUpdateAnimation()
   > n2 = Sequence!(execute) -> (then_0, then_1, then_2:Exec)
   > Set Delta Time X(Delta Time X=@n1.DeltaTimeX)
   n3 = AND Boolean(A=Get bIsCrouching(), B=NOT Boolean(A=Get bIsMoving2D()))
   > Set CrouchIdleSwordRotateAlpha(CrouchIdleSwordRotateAlpha:float=DampFloat!(
     Value:float=Get CrouchIdleSwordRotateAlpha() -> (CrouchIdleSwordRotateAlpha:float)
     Target:float=SelectFloat(A=1, B=0, bPickA=@n3)
     DeltaTime:float=Get Delta Time X()
     SmoothingTime:float=0.2
    ) -> (ReturnValue:float)) -> (Output_Get:float)
   n4 = AND Boolean(A=CheckForBackbling().bIsBackblingLong, B=NOT Boolean(A=Get bIsCrouching()))
   > Set IdleWithCapeSwordRotateAlpha(IdleWithCapeSwordRotateAlpha:float=DampFloat(
     Value=Get IdleWithCapeSwordRotateAlpha() -> (IdleWithCapeSwordRotateAlpha:float)
     Target=SelectFloat(A=1, B=0, bPickA=@n4)
     DeltaTime=Get Delta Time X()
     SmoothingTime=0.2
    )) -> (Output_Get:float)
   Branch(execute=@n2.then_1, Condition=Get bIsInFrontEnd())
   > Set customIdleIsPlayingFloat(customIdleIsPlayingFloat:float=GetCurveValue(self=GetOwnerFrontEndAnimInstance(), CurveName=Frontend_Idle_Bypass)) -> (Output_Get:float)
   n5 = Get bIsTargeting()
   Set IsTargetAlpha(
    execute=@n2.then_2
    IsTargetAlpha:double=DampFloat(Value=Get IsTargetAlpha(), Target=SelectFloat(A=1, B=0, bPickA=@n5), DeltaTime=Get Delta Time X(), SmoothingTime=0.05)
   ) -> (Output_Get:double)
  ]}
]
variables [
 CrouchIdleSwordRotateAlpha {type=number}
 customIdleIsPlayingFloat {type=number}
 Delta Time X {type=number}
 IdleWithCapeSwordRotateAlpha {type=number}
 IsTargetAlpha {type=number}
]
